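# Sound cue for the looping ship sail ambience in the Skyfire arena, played as positional ambient sound.
# A looping wave player feeds an Enveloper that ramps volume from silent to full at the start, then a Quality Level node with only its first input connected.
# An attenuation asset shapes distance falloff, and a concurrency override allows a single instance, stopping the quietest on conflict.
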SoundCue /Skyfire/Sound/Skyfire_Arena_Ambience_Ship_Sail_Loop_Cue
root=QualityLevel_0
nodes [
 {
  id=Enveloper_0
  children [WavePlayer_0]
  properties {pitchCurve {editorCurveData {keys [{time=0}]}}, volumeCurve {editorCurveData {keys [<at 0 0>,<at 1 1>]}}}
 }
 {id=QualityLevel_0, children [Enveloper_0,null,null,null]}
 {id=WavePlayer_0, properties {bLooping=true, soundWaveAssetPtr=&/Skyfire_Arena_Ambience_Ship_Sail_Loop}}
]
properties {
 attenuationSettings=&/Skyfire_Ship_Sail_Loop_Att
 bOverrideConcurrency=true
 maxDistance=3247.4998
 priority=0
 soundClassObject=/Game/Sounds/FortSoundClassesAndModes/Fort_LTM_BR_InGame_Phase_Main_Amb_3D
 volumeMultiplier=2
 concurrencyOverrides {maxCount=1, resolutionRule=StopQuietest}
}
