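# Sound cue for the looping ambient stasis sound of a floating object in the Skyfire arena.
# It holds settings only: the root is empty and the file shows no sound nodes or waves.
# It references its own attenuation asset and a Fortnite LTM 3D ambience sound class.
# A concurrency override limits it to a single instance, stopping the quietest when exceeded.
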SoundCue /Skyfire/Sound/Skyfire_Arena_Floating_Object_Stasis_Loop_Cue
root=null
nodes []
properties {
 attenuationSettings=&/Skyfire_Arena_Floating_Object_Stasis_Loop_Att
 bOverrideConcurrency=true
 maxDistance=2700
 soundClassObject=/Game/Sounds/FortSoundClassesAndModes/Fort_LTM_BR_InGame_Phase_Main_Amb_3D
 volumeMultiplier=1
 concurrencyOverrides {maxCount=1, resolutionRule=StopQuietest}
}
