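# Looping 3D ambient sound cue for the laser net grid on the Skyfire arena floor.
# A looping wave feeds a continuous modulator, where the GridPitching parameter drives pitch and volume through inverted ranges.
# A quality-level root routes that branch to its first two slots and leaves the other two empty.
# An external attenuation asset is assigned, and a concurrency override allows one instance, stopping the quietest.
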SoundCue /Skyfire/Sound/Skyfire_Arena_Floor_Net_Grid_Laser_Loop_Cue
root=QualityLevel_0
nodes [
 {
  id=ModulatorContinuous_0
  children [WavePlayer_0]
  properties {
   pitchModulationParams {maxOutput=0.8, minOutput=1.3, parameterName=GridPitching}
   volumeModulationParams {maxInput=0, minInput=1, minOutput=0.84, parameterName=GridPitching}
  }
 }
 {id=QualityLevel_0, children [ModulatorContinuous_0, ModulatorContinuous_0, null, null]}
 {id=WavePlayer_0, properties {bLooping=true, soundWaveAssetPtr=&/Skyfire_Arena_Floor_Net_Grid_Laser_Loop}}
]
properties {
 attenuationSettings=&/Skyfire_Arena_Floor_Net_Grid_Laser_Loop_Att
 bOverrideConcurrency=true
 maxDistance=5200
 priority=70
 soundClassObject=/Game/Sounds/FortSoundClassesAndModes/Fort_LTM_BR_InGame_Phase_Main_Amb_3D
 volumeMultiplier=5
 attenuationOverrides {
  bAttenuate=false
  distanceAlgorithm=Custom
  falloffDistance=4800
  nonSpatializedRadiusStart=1500
  stereoSpread=0
  customAttenuationCurve {editorCurveData {keys [{arriveTangent=-2.2385, interpMode=RCIM_Cubic, leaveTangent=-2.2385, tangentMode=RCTM_User, time=0.0088, value=1.0422}, <at 1 0>]}}
 }
 concurrencyOverrides {maxCount=1, resolutionRule=StopQuietest}
}
