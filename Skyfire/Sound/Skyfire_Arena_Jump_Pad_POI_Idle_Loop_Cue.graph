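# Sound cue for the looping idle ambience of the Skyfire arena jump pad point of interest.
# A looping wave player feeds a Modulator node at default settings, which varies the sound on playback.
# It plays through a 3D ambience sound class with its own attenuation asset and boosted volume.
# A concurrency override caps simultaneous instances at a few, stopping the quietest with a short release.
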SoundCue /Skyfire/Sound/Skyfire_Arena_Jump_Pad_POI_Idle_Loop_Cue
root=Modulator_0
nodes [
 {id=Modulator_0, children [WavePlayer_0]}
 {id=WavePlayer_0, properties {bLooping=true, soundWaveAssetPtr=&/Skyfire_Arena_Jump_Pad_Idle_Loop}}
]
properties {
 attenuationSettings=&/Skyfire_Arena_Jump_Pad_Idle_Loop_Att
 bOverrideConcurrency=true
 maxDistance=4200
 priority=30
 soundClassObject=/Game/Sounds/FortSoundClassesAndModes/Fort_LTM_BR_InGame_Phase_Main_Amb_3D
 volumeMultiplier=2
 concurrencyOverrides {maxCount=3, resolutionRule=StopQuietest, voiceStealReleaseTime=0.15}
}
