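# Sound cue for the launch of a Skyfire arena jump pad POI, playing a whoosh as a 3D sound.
# A single wave player feeds a Modulator left at default settings, with boosted cue volume and a dedicated attenuation asset.
# It overrides concurrency to cap simultaneous voices, stopping the quietest first, and disables virtualization.
# It routes to an LTM in-game main-phase 3D ambience sound class.
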SoundCue /Skyfire/Sound/Skyfire_Arena_Jump_Pad_POI_Launch_Cue
root=Modulator_0
nodes [
 {id=Modulator_0, children [WavePlayer_0]}
 {id=WavePlayer_0, properties {soundWaveAssetPtr=&/Skyfire_Arena_Jump_Pad_Launch_Whoosh_01}}
]
properties {
 attenuationSettings=&/Skyfire_Arena_Jump_Pad_Launch_Att
 bOverrideConcurrency=true
 maxDistance=8500
 priority=50
 soundClassObject=/Game/Sounds/FortSoundClassesAndModes/Fort_LTM_BR_InGame_Phase_Main_Amb_3D
 virtualizationMode=Disabled
 volumeMultiplier=2.5
 concurrencyOverrides {maxCount=3, resolutionRule=StopQuietest, voiceStealReleaseTime=0.5}
}
