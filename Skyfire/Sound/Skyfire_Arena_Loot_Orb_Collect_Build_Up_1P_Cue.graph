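# Sound cue for the first-person build-up audio of collecting a loot orb in the Skyfire arena.
# It plays a single wave asset directly, with boosted volume and high priority that bypasses volume scaling, routed through an LTM in-game ambient 3D sound class.
# Virtualization is disabled and a concurrency override keeps only one instance playing, stopping the oldest when retriggered.
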SoundCue /Skyfire/Sound/Skyfire_Arena_Loot_Orb_Collect_Build_Up_1P_Cue
root=WavePlayer_0
nodes [{id=WavePlayer_0, properties {soundWaveAssetPtr=&/Skyfire_Arena_Loot_Orb_Collect_Build_Up_01}}]
properties {
 bBypassVolumeScaleForPriority=true
 bOverrideConcurrency=true
 maxDistance=2097152
 priority=80
 soundClassObject=/Game/Sounds/FortSoundClassesAndModes/Fort_LTM_BR_InGame_Phase_Main_Amb_3D
 virtualizationMode=Disabled
 volumeMultiplier=1.5
 concurrencyOverrides {maxCount=1, resolutionRule=StopOldest}
}
