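# Sound cue for collecting a loot orb in the Skyfire arena, playing a single burst wave directly from one wave player with no randomization or modulation.
# It is spatialized through its own attenuation asset and routed to an LTM in-game main-phase 3D ambience sound class.
# A concurrency override allows only one instance, stopping the quietest with a release fade, and virtualization is disabled.
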SoundCue /Skyfire/Sound/Skyfire_Arena_Loot_Orb_Collect_Cue
root=WavePlayer_0
nodes [{id=WavePlayer_0, properties {soundWaveAssetPtr=&/Skyfire_Arena_Loot_Orb_Collect_Burst_01}}]
properties {
 attenuationSettings=&/Skyfire_Arena_Loot_Orb_Collect_Att
 bOverrideConcurrency=true
 maxDistance=6500
 priority=25
 soundClassObject=/Game/Sounds/FortSoundClassesAndModes/Fort_LTM_BR_InGame_Phase_Main_Amb_3D
 virtualizationMode=Disabled
 volumeMultiplier=1
 concurrencyOverrides {maxCount=1, resolutionRule=StopQuietest, voiceStealReleaseTime=1}
}
