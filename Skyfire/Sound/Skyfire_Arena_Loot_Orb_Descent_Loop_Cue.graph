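# Sound cue for the looping descent sound of a Skyfire arena loot orb, played as spatialized ambience in the LTM in-game 3D sound class.
# A looping wave feeds a default modulator, a two-second envelope fade-in and an attenuation node, then a boosted mixer input; the mixer's other input is empty.
# A concurrency override caps it at three instances, stopping the quietest.
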SoundCue /Skyfire/Sound/Skyfire_Arena_Loot_Orb_Descent_Loop_Cue
root=Mixer_0
nodes [
 {id=Attenuation_1, children [Enveloper_0], properties {attenuationSettings=&/Skyfire_Arena_Loot_Orb_Descent_Loop_Att}}
 {
  id=Enveloper_0
  children [Modulator_1]
  properties {pitchCurve {editorCurveData {keys [{time=0}]}}, volumeCurve {editorCurveData {keys [<at 0 0 RCIM_Cubic>,<at 2 1>]}}}
 }
 {id=Mixer_0, children [null, Attenuation_1], properties {inputVolume [1,1.5]}}
 {id=Modulator_1, children [WavePlayer_0]}
 {id=WavePlayer_0, properties {bLooping=true, soundWaveAssetPtr=&/Skyfire_Arena_Loot_Orb_Descent_Loop}}
]
properties {
 bOverrideConcurrency=true
 maxDistance=12300
 priority=40
 soundClassObject=/Game/Sounds/FortSoundClassesAndModes/Fort_LTM_BR_InGame_Phase_Main_Amb_3D
 volumeMultiplier=1.3
 concurrencyOverrides {maxCount=3, resolutionRule=StopQuietest}
}
