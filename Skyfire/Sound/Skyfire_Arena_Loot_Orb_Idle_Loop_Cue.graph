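# Sound cue for the idle ambience of the Skyfire arena loot orb, layering a one-shot start with a sustained loop.
# A mixer combines a random pick of four start waves, shared by all quality levels, with a looping wave that follows a short delay and fades in through an enveloper.
# Each layer has its own modulator and attenuation, and overridden concurrency caps the cue at three instances, stopping the quietest.
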SoundCue /Skyfire/Sound/Skyfire_Arena_Loot_Orb_Idle_Loop_Cue
root=Mixer_0
nodes [
 {id=Attenuation_0, children [Modulator_1], properties {attenuationSettings=&/Skyfire_Arena_Loot_Orb_Idle_Start_Att}}
 {id=Attenuation_1, children [Delay_0], properties {attenuationSettings=&/Skyfire_Arena_Loot_Orb_Idle_Loop_Att}}
 {id=Delay_0, children [Enveloper_0], properties {delayMax=0.15, delayMin=0.15}}
 {
  id=Enveloper_0
  children [Modulator_0]
  properties {pitchCurve {editorCurveData {keys [{time=0}]}}, volumeCurve {editorCurveData {keys [<at 0 0 RCIM_Cubic>,<at 2 1>]}}}
 }
 {id=Mixer_0, children [QualityLevel_0, Attenuation_1], properties {inputVolume [1.75,2]}}
 {id=Modulator_0, children [WavePlayer_0]}
 {id=Modulator_1, children [Random_0]}
 {id=QualityLevel_0, children [Attenuation_0, Attenuation_0, Attenuation_0, Attenuation_0]}
 {id=Random_0, children [WavePlayer_1,WavePlayer_2,WavePlayer_3,WavePlayer_4], properties {weights [1,1,1,1]}}
 {id=WavePlayer_0, properties {bLooping=true, soundWaveAssetPtr=&/Skyfire_Arena_Loot_Orb_Idle_Loop}}
 {id=WavePlayer_1, properties {soundWaveAssetPtr=&/Skyfire_Arena_Loot_Orb_Idle_Start_04}}
 {id=WavePlayer_2, properties {soundWaveAssetPtr=&/Skyfire_Arena_Loot_Orb_Idle_Start_01}}
 {id=WavePlayer_3, properties {soundWaveAssetPtr=&/Skyfire_Arena_Loot_Orb_Idle_Start_02}}
 {id=WavePlayer_4, properties {soundWaveAssetPtr=&/Skyfire_Arena_Loot_Orb_Idle_Start_03}}
]
properties {
 bOverrideConcurrency=true
 maxDistance=9291.999
 priority=45
 soundClassObject=/Game/Sounds/FortSoundClassesAndModes/Fort_LTM_BR_InGame_Phase_Main_Amb_3D
 volumeMultiplier=1
 concurrencyOverrides {maxCount=3, resolutionRule=StopQuietest}
}
