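# Sound cue that plays the buzzer for a loot orb spawn in the Skyfire arena.
# A single drop-buzzer wave feeds a Mixer that boosts its volume; the Mixer's second input is empty.
# A concurrency override allows one instance and blocks new ones, and the cue keeps playing when silent.
# High priority, a very large max distance and an LTM main-phase 3D ambience sound class are set.
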SoundCue /Skyfire/Sound/Skyfire_Arena_Loot_Orb_Spawn_Buzzer_Cue
root=Mixer_0
nodes [
 {id=Mixer_0, children [WavePlayer_0,null], properties {inputVolume [1.65,1.8]}}
 {id=WavePlayer_0, properties {soundWaveAssetPtr=&/Skyfire_Arena_Loot_Orb_Drop_Buzzer_01_C}}
]
properties {
 bBypassVolumeScaleForPriority=true
 bOverrideConcurrency=true
 maxDistance=2097152
 priority=85
 soundClassObject=/Game/Sounds/FortSoundClassesAndModes/Fort_LTM_BR_InGame_Phase_Main_Amb_3D
 virtualizationMode=PlayWhenSilent
 volumeMultiplier=1
 concurrencyOverrides {maxCount=1, resolutionRule=PreventNew}
}
