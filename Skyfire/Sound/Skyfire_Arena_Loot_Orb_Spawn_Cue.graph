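# Sound cue for a loot orb spawning in the Skyfire arena.
# A Random node picks one of four equally weighted spawn waves, and a Modulator varies its pitch upward only, leaving volume constant.
# It uses its own attenuation asset and a 3D LTM ambience sound class, with raised priority and virtualization disabled.
# A concurrency override caps it at two simultaneous instances, stopping the quietest.
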SoundCue /Skyfire/Sound/Skyfire_Arena_Loot_Orb_Spawn_Cue
root=Modulator_0
nodes [
 {id=Modulator_0, children [Random_0], properties {pitchMax=1.25, pitchMin=1, volumeMax=1, volumeMin=1}}
 {id=Random_0, children [WavePlayer_1,WavePlayer_2,WavePlayer_3,WavePlayer_4], properties {weights [1,1,1,1]}}
 {id=WavePlayer_1, properties {soundWaveAssetPtr=&/Skyfire_Arena_Loot_Orb_Spawn_01}}
 {id=WavePlayer_2, properties {soundWaveAssetPtr=&/Skyfire_Arena_Loot_Orb_Spawn_02}}
 {id=WavePlayer_3, properties {soundWaveAssetPtr=&/Skyfire_Arena_Loot_Orb_Spawn_03}}
 {id=WavePlayer_4, properties {soundWaveAssetPtr=&/Skyfire_Arena_Loot_Orb_Spawn_04}}
]
properties {
 attenuationSettings=&/Skyfire_Arena_Loot_Orb_Spawn_Att
 bBypassVolumeScaleForPriority=true
 bOverrideConcurrency=true
 maxDistance=2097152
 priority=80
 soundClassObject=/Game/Sounds/FortSoundClassesAndModes/Fort_LTM_BR_InGame_Phase_Main_Amb_3D
 virtualizationMode=Disabled
 volumeMultiplier=1.2
 concurrencyOverrides {maxCount=2, resolutionRule=StopQuietest, voiceStealReleaseTime=0.25}
}
