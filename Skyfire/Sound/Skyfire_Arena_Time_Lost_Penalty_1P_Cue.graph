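# Sound cue for the first-person stinger played when the player receives a time-lost penalty in the Skyfire arena mode.
# A single wave player plays the penalty stinger wave through an in-game phase sound class with a slight volume boost.
# Attenuation and spatialization are disabled, so it plays as a non-positional 2D sound.
# It has high priority without virtualization, and concurrency allows one instance, stopping the oldest when retriggered.
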SoundCue /Skyfire/Sound/Skyfire_Arena_Time_Lost_Penalty_1P_Cue
root=WavePlayer_0
nodes [{id=WavePlayer_0, properties {soundWaveAssetPtr=&/Skyfire_Arena_Time_Lost_Penalty_Stinger}}]
properties {
 bBypassVolumeScaleForPriority=true
 bOverrideAttenuation=true
 bOverrideConcurrency=true
 maxDistance=2097152
 priority=85
 soundClassObject=/Game/Sounds/FortSoundClassesAndModes/Fort_LTM_BR_InGame_Phase_Main_Amb_3D
 virtualizationMode=Disabled
 volumeMultiplier=1.2
 attenuationOverrides {bAttenuate=false, bSpatialize=false}
 concurrencyOverrides {maxCount=1, resolutionRule=StopOldest}
}
