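# Sound cue for the Skyfire arena time-lost penalty, playing one of six explosion variations.
# An equal-weight Random node picks the wave, feeding a quality-level switch with only its first two slots connected, then a Modulator that varies pitch.
# The boosted cue uses its own attenuation asset and an LTM 3D ambience sound class.
# It allows one instance, stopping the quietest, with virtualization disabled and zero priority.
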SoundCue /Skyfire/Sound/Skyfire_Arena_Time_Lost_Penalty_Cue
root=Modulator_1
nodes [
 {id=Modulator_1, children [QualityLevel_1], properties {pitchMax=1.2}}
 {id=QualityLevel_1, children [Random_2,Random_2,null,null]}
 {
  id=Random_2
  children [WavePlayer_10, WavePlayer_11, WavePlayer_12, WavePlayer_13, WavePlayer_14, WavePlayer_15]
  properties {bShouldExcludeFromBranchCulling=true, weights [1,1,1,1,1,1]}
 }
 {id=WavePlayer_10, properties {soundWaveAssetPtr=&/Skyfire_Arena_Time_Lost_Penalty_Explo_01}}
 {id=WavePlayer_11, properties {soundWaveAssetPtr=&/Skyfire_Arena_Time_Lost_Penalty_Explo_02}}
 {id=WavePlayer_12, properties {soundWaveAssetPtr=&/Skyfire_Arena_Time_Lost_Penalty_Explo_03}}
 {id=WavePlayer_13, properties {soundWaveAssetPtr=&/Skyfire_Arena_Time_Lost_Penalty_Explo_04}}
 {id=WavePlayer_14, properties {soundWaveAssetPtr=&/Skyfire_Arena_Time_Lost_Penalty_Explo_05}}
 {id=WavePlayer_15, properties {soundWaveAssetPtr=&/Skyfire_Arena_Time_Lost_Penalty_Explo_06}}
]
properties {
 attenuationSettings=&/Skyfire_Arena_Time_Lost_Penalty_Death_Att
 bOverrideConcurrency=true
 maxDistance=10500
 priority=0
 soundClassObject=/Game/Sounds/FortSoundClassesAndModes/Fort_LTM_BR_InGame_Phase_Main_Amb_3D
 virtualizationMode=Disabled
 volumeMultiplier=3
 concurrencyOverrides {maxCount=1, resolutionRule=StopQuietest}
}
