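# Sound cue for the Skyfire arena, named for a time-lost penalty on respawn, that plays a single wave with attenuation.
# A modulator randomizes the wave's pitch, and a quality-level root routes to it on the first two levels, leaving the other two empty.
# The cue boosts volume, uses a 3D LTM ambience sound class and lowest priority, and disables virtualization.
# A concurrency override allows only one instance, stopping the quietest.
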SoundCue /Skyfire/Sound/Skyfire_Arena_Time_Lost_Penalty_Respawn_Cue
root=QualityLevel_0
nodes [
 {id=Modulator_1, children [WavePlayer_9], properties {pitchMax=1.2}}
 {id=QualityLevel_0, children [Modulator_1,Modulator_1,null,null]}
 {id=WavePlayer_9, properties {soundWaveAssetPtr=&/Skyfire_Arena_Time_Lost_Penalty_Respawn_01}}
]
properties {
 attenuationSettings=&/Skyfire_Arena_Time_Lost_Penalty_Respawn_Att
 bOverrideConcurrency=true
 maxDistance=13000
 priority=0
 soundClassObject=/Game/Sounds/FortSoundClassesAndModes/Fort_LTM_BR_InGame_Phase_Main_Amb_3D
 virtualizationMode=Disabled
 volumeMultiplier=2.5
 concurrencyOverrides {maxCount=1, resolutionRule=StopQuietest}
}
