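# Sound cue for collecting a time orb in the Skyfire arena, with 1P in the name suggesting the first-person variant.
# A single wave player feeds a modulator that varies pitch per play, with a raised upper limit.
# A concurrency override allows one instance at a time, stopping the oldest.
# It has boosted volume, raised priority that bypasses volume scaling, and routes to an LTM in-game 3D sound class.
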SoundCue /Skyfire/Sound/Skyfire_Arena_Time_Orb_Collect_1P_Cue
root=Modulator_0
nodes [
 {id=Modulator_0, children [WavePlayer_0], properties {pitchMax=1.2}}
 {id=WavePlayer_0, properties {soundWaveAssetPtr=&/Skyfire_Arena_Time_Orb_Collect_01}}
]
properties {
 bBypassVolumeScaleForPriority=true
 bOverrideConcurrency=true
 maxDistance=2097152
 priority=85
 soundClassObject=/Game/Sounds/FortSoundClassesAndModes/Fort_LTM_BR_InGame_Phase_Main_Amb_3D
 volumeMultiplier=1.8
 concurrencyOverrides {maxCount=1, resolutionRule=StopOldest}
}
