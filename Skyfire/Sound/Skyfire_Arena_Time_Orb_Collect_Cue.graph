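# Sound cue for collecting a time orb in the Skyfire arena, spatialized through its own attenuation asset.
# A single wave feeds a Modulator that randomizes a raised pitch, which a Quality Level root uses for its first two levels, leaving the other two empty.
# A concurrency override allows one instance at a time, stopping the quietest, and virtualization is disabled.
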SoundCue /Skyfire/Sound/Skyfire_Arena_Time_Orb_Collect_Cue
root=QualityLevel_0
nodes [
 {id=Modulator_0, children [WavePlayer_0], properties {pitchMax=1.35, pitchMin=1.25}}
 {id=QualityLevel_0, children [Modulator_0,Modulator_0,null,null]}
 {id=WavePlayer_0, properties {soundWaveAssetPtr=&/Skyfire_Arena_Time_Orb_Collect_01}}
]
properties {
 attenuationSettings=&/Skyfire_Arena_Time_Orb_Collect_Att
 bOverrideConcurrency=true
 maxDistance=1900
 priority=25
 soundClassObject=/Game/Sounds/FortSoundClassesAndModes/Fort_LTM_BR_InGame_Phase_Main_Amb_3D
 virtualizationMode=Disabled
 volumeMultiplier=1
 concurrencyOverrides {maxCount=1, resolutionRule=StopQuietest}
}
