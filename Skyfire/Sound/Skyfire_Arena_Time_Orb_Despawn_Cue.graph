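# Sound cue for the time orb despawn effect in the Skyfire Arena, routed to an LTM in-game 3D ambience sound class.
# A single despawn wave passes through a default Modulator, which fills the first two Quality Level slots; the other two are empty.
# It reuses the orb spawn attenuation and overrides concurrency to a small cap that stops the quietest instance, with virtualization disabled.
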SoundCue /Skyfire/Sound/Skyfire_Arena_Time_Orb_Despawn_Cue
root=QualityLevel_0
nodes [
 {id=Modulator_0, children [WavePlayer_0]}
 {id=QualityLevel_0, children [Modulator_0,Modulator_0,null,null]}
 {id=WavePlayer_0, properties {soundWaveAssetPtr=&/Skyfire_Arena_Time_Orb_Despawn_01}}
]
properties {
 attenuationSettings=&/Skyfire_Arena_Time_Orb_Spawn_Att
 bOverrideConcurrency=true
 maxDistance=8300
 priority=0
 soundClassObject=/Game/Sounds/FortSoundClassesAndModes/Fort_LTM_BR_InGame_Phase_Main_Amb_3D
 virtualizationMode=Disabled
 volumeMultiplier=1.2
 concurrencyOverrides {maxCount=2, resolutionRule=StopQuietest}
}
