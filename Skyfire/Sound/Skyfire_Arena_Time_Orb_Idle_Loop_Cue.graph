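# Sound cue for the looping idle sound of the Time Orb in the Skyfire Arena, played as a 3D ambient LTM sound with its own attenuation.
# A looping wave player feeds a modulator that randomizes pitch, then an enveloper that ramps volume in over a quarter second.
# A concurrency override limits simultaneous instances to a few, stopping the quietest with a short release.
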SoundCue /Skyfire/Sound/Skyfire_Arena_Time_Orb_Idle_Loop_Cue
root=Enveloper_0
nodes [
 {
  id=Enveloper_0
  children [Modulator_0]
  properties {pitchCurve {editorCurveData {keys [{time=0}]}}, volumeCurve {editorCurveData {keys [<at 0 0>,<at 0.25 1>]}}}
 }
 {id=Modulator_0, children [WavePlayer_0], properties {pitchMax=1.1}}
 {id=WavePlayer_0, properties {bLooping=true, soundWaveAssetPtr=&/Skyfire_Arena_Time_Orb_Idle_Loop}}
]
properties {
 attenuationSettings=&/Skyfire_Arena_Time_Orb_Idle_Loop_Att
 bOverrideConcurrency=true
 maxDistance=6300
 priority=50
 soundClassObject=/Game/Sounds/FortSoundClassesAndModes/Fort_LTM_BR_InGame_Phase_Main_Amb_3D
 volumeMultiplier=3
 concurrencyOverrides {maxCount=3, resolutionRule=StopQuietest, voiceStealReleaseTime=0.1}
}
