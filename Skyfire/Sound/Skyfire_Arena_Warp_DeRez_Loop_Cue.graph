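# Sound cue for the Skyfire arena warp de-rez loop, pairing a start tell with a sustained loop.
# A mixer plays the louder, non-looping tell with randomized pitch alongside a looping wave that an enveloper fades in over about a second, also pitch-modulated.
# It uses character de-rez loop attenuation and the LTM special item sound class, with a concurrency override that stops the quietest instance beyond a small limit.
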SoundCue /Skyfire/Sound/Skyfire_Arena_Warp_DeRez_Loop_Cue
root=Mixer_0
nodes [
 {
  id=Enveloper_0
  children [WavePlayer_0]
  properties {
   pitchCurve {editorCurveData {keys [{time=0}]}}
   volumeCurve {editorCurveData {keys [{arriveTangent=-0.0038, interpMode=RCIM_Cubic, leaveTangent=-0.0038, tangentMode=RCTM_User, time=0, value=0}, <at 1.25 1>]}}
  }
 }
 {id=Mixer_0, children [Modulator_1,Modulator_0], properties {inputVolume [1.5,1]}}
 {id=Modulator_0, children [Enveloper_0], properties {volumeMax=1, volumeMin=1}}
 {id=Modulator_1, children [WavePlayer_1], properties {pitchMax=1.2, volumeMax=1, volumeMin=1}}
 {id=WavePlayer_0, properties {bLooping=true, soundWaveAssetPtr=&/Skyfire_Warp_DeRez_Loop}}
 {id=WavePlayer_1, properties {soundWaveAssetPtr=&/Skyfire_Warp_DeRez_Tell_Loop_Start_01}}
]
properties {
 attenuationSettings=&/Skyfire_Warp_Character_DeRez_Loop_Att
 bOverrideConcurrency=true
 maxDistance=6200
 priority=75
 soundClassObject=/Game/Sounds/FortSoundClassesAndModes/Fort_LTM_BR_InGame_Special_Item
 volumeMultiplier=1.5
 concurrencyOverrides {maxCount=3, resolutionRule=StopQuietest}
}
