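# Sound cue for the default first-person countdown tick of the Skyfire backpack, routed to the first-person LTM special item sound class.
# A quality-level node feeds the same tick wave player into its two filled slots, with the cue raising pitch and lowering volume.
# An inline concurrency override allows one instance at a time, stopping the oldest, and virtualization is disabled.
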SoundCue /Skyfire/Sound/Skyfire_Backpack_Countdown_Tick_Default_1P_Cue
root=QualityLevel_0
nodes [
 {id=QualityLevel_0, children [WavePlayer_1,WavePlayer_1,null,null]}
 {id=WavePlayer_1, properties {soundWaveAssetPtr=&/Skyfire_Backpack_Countdown_Tick_Default}}
]
properties {
 bBypassVolumeScaleForPriority=true
 bOverrideConcurrency=true
 maxDistance=2097152
 pitchMultiplier=2
 priority=0
 soundClassObject=/Game/Sounds/FortSoundClassesAndModes/Fort_LTM_BR_InGame_Special_Item_1P
 virtualizationMode=Disabled
 volumeMultiplier=0.25
 concurrencyOverrides {maxCount=1, resolutionRule=StopOldest}
}
