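# Sound cue named as the default countdown tick for the Skyfire backpack, with no nodes and a null root, so it plays no audio.
# It keeps only playback settings: the backpack attenuation asset, a maximum distance, and the Fortnite LTM in-game special item sound class.
# Virtualization is disabled.
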SoundCue /Skyfire/Sound/Skyfire_Backpack_Countdown_Tick_Default_Cue
root=null
nodes []
properties {
 attenuationSettings=&/Skyfire_Backpack_Att
 maxDistance=1000
 soundClassObject=/Game/Sounds/FortSoundClassesAndModes/Fort_LTM_BR_InGame_Special_Item
 virtualizationMode=Disabled
}
