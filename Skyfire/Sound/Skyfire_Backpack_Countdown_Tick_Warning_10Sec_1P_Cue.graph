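# Sound cue for the Skyfire backpack's first-person 10-second countdown tick warning.
# A mixer layers warning beep B, a heavily pitched-up countdown tick and beep A, which a short delay offsets into a staggered beep.
# It uses a first-person special-item sound class with raised priority, disabled virtualization and a concurrency override that allows one instance, stopping the oldest.
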SoundCue /Skyfire/Sound/Skyfire_Backpack_Countdown_Tick_Warning_10Sec_1P_Cue
root=Mixer_0
nodes [
 {id=Delay_0, children [WavePlayer_2], properties {delayMax=0.1, delayMin=0.1}}
 {id=Mixer_0, children [WavePlayer_1,Modulator_0,Delay_0], properties {inputVolume [1,0.95,1]}}
 {id=Modulator_0, children [WavePlayer_0], properties {pitchMax=3, pitchMin=3, volumeMax=1, volumeMin=1}}
 {id=WavePlayer_0, properties {soundWaveAssetPtr=&/Skyfire_Backpack_Countdown_Tick_Warning}}
 {id=WavePlayer_1, properties {soundWaveAssetPtr=&/Skyfire_Backpack_Warning_Beep_B}}
 {id=WavePlayer_2, properties {soundWaveAssetPtr=&/Skyfire_Backpack_Warning_Beep_A}}
]
properties {
 bBypassVolumeScaleForPriority=true
 bOverrideConcurrency=true
 maxDistance=2097152
 priority=75
 soundClassObject=/Game/Sounds/FortSoundClassesAndModes/Fort_LTM_BR_InGame_Special_Item_1P
 virtualizationMode=Disabled
 volumeMultiplier=1.2
 concurrencyOverrides {maxCount=1, resolutionRule=StopOldest}
}
