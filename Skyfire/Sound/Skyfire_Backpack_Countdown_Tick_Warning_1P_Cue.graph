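# First-person sound cue for the warning tick of the Skyfire backpack countdown.
# A single wave player feeds the root directly, with the pitch multiplier raised well above normal.
# It routes to the LTM first-person special-item sound class with raised priority that bypasses volume scaling, and virtualization is disabled.
# A concurrency override allows one instance at a time, so each new tick stops the previous one.
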SoundCue /Skyfire/Sound/Skyfire_Backpack_Countdown_Tick_Warning_1P_Cue
root=WavePlayer_0
nodes [{id=WavePlayer_0, properties {soundWaveAssetPtr=&/Skyfire_Backpack_Countdown_Tick_Warning}}]
properties {
 bBypassVolumeScaleForPriority=true
 bOverrideConcurrency=true
 maxDistance=2097152
 pitchMultiplier=3
 priority=20
 soundClassObject=/Game/Sounds/FortSoundClassesAndModes/Fort_LTM_BR_InGame_Special_Item_1P
 virtualizationMode=Disabled
 volumeMultiplier=0.95
 concurrencyOverrides {maxCount=1, resolutionRule=StopOldest}
}
