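# Sound cue for the first-person start of the Skyfire backpack countup, playing the prison timer init wave.
# A fixed short delay feeds a single wave player, with boosted volume and the backpack attenuation asset.
# It routes to the LTM in-game master sound class at high priority with virtualization disabled.
# Concurrency is overridden to one instance, stopping the oldest when retriggered.
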SoundCue /Skyfire/Sound/Skyfire_Backpack_Countup_Start_1P_Cue
root=Delay_0
nodes [
 {id=Delay_0, children [WavePlayer_0], properties {delayMax=0.1, delayMin=0.1}}
 {id=WavePlayer_0, properties {soundWaveAssetPtr=&/Skyfire_Backpack_Timer_Prison_Init_01}}
]
properties {
 attenuationSettings=&/Skyfire_Backpack_Att
 bBypassVolumeScaleForPriority=true
 bOverrideConcurrency=true
 maxDistance=1350
 priority=75
 soundClassObject=/Game/Sounds/FortSoundClassesAndModes/Fort_LTM_BR_InGame_Master
 virtualizationMode=Disabled
 volumeMultiplier=3.5
 concurrencyOverrides {maxCount=1, resolutionRule=StopOldest}
}
