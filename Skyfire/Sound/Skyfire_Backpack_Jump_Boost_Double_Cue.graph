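# Sound cue for the Skyfire jetpack backpack's double jump boost, in a traversal sound class with its own attenuation.
# A quality-level switch feeds the first two levels from six random boost variations and the last two from three.
# Each random pool passes through a modulator for playback variation and is excluded from branch culling.
# A concurrency override allows two instances, stopping the quietest beyond that.
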SoundCue /Skyfire/Sound/Skyfire_Backpack_Jump_Boost_Double_Cue
root=QualityLevel_0
nodes [
 {id=Modulator_0, children [Random_0]}
 {id=Modulator_1, children [Random_1]}
 {id=QualityLevel_0, children [Modulator_0,Modulator_0,Modulator_1,Modulator_1]}
 {
  id=Random_0
  children [WavePlayer_1,WavePlayer_2,WavePlayer_3,WavePlayer_4,WavePlayer_5,WavePlayer_6]
  properties {bShouldExcludeFromBranchCulling=true, weights [1,1,1,1,1,1]}
 }
 {id=Random_1, children [WavePlayer_0,WavePlayer_7,WavePlayer_8], properties {bShouldExcludeFromBranchCulling=true, weights [1,1,1]}}
 {id=WavePlayer_0, properties {soundWaveAssetPtr=&/Skyfire_Backpack_Jump_Jetpack_Double_Boost_01}}
 {id=WavePlayer_1, properties {soundWaveAssetPtr=&Skyfire_Backpack_Jump_Jetpack_Double_Boost_01}}
 {id=WavePlayer_2, properties {soundWaveAssetPtr=&/Skyfire_Backpack_Jump_Jetpack_Double_Boost_02}}
 {id=WavePlayer_3, properties {soundWaveAssetPtr=&/Skyfire_Backpack_Jump_Jetpack_Double_Boost_03}}
 {id=WavePlayer_4, properties {soundWaveAssetPtr=&/Skyfire_Backpack_Jump_Jetpack_Double_Boost_04}}
 {id=WavePlayer_5, properties {soundWaveAssetPtr=&/Skyfire_Backpack_Jump_Jetpack_Double_Boost_05}}
 {id=WavePlayer_6, properties {soundWaveAssetPtr=&/Skyfire_Backpack_Jump_Jetpack_Double_Boost_06}}
 {id=WavePlayer_7, properties {soundWaveAssetPtr=&Skyfire_Backpack_Jump_Jetpack_Double_Boost_02}}
 {id=WavePlayer_8, properties {soundWaveAssetPtr=&Skyfire_Backpack_Jump_Jetpack_Double_Boost_03}}
]
properties {
 attenuationSettings=&/Skyfire_Backpack_Jump_Boost_Double_Att
 bOverrideConcurrency=true
 maxDistance=10800
 priority=80
 soundClassObject=/Game/Sounds/FortSoundClassesAndModes/Fort_LTM_BR_InGame_Traversal
 volumeMultiplier=1.2
 concurrencyOverrides {maxCount=2, resolutionRule=StopQuietest}
}
