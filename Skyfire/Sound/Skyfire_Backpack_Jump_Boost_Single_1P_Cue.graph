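# Skyfire sound cue for the first-person single jump boost of the jetpack backpack, routed to the 1P traversal sound class.
# A quality-level root sends its first two slots to a random pick of six boost waves and its last two to a pick of the first three, each through a Modulator.
# Virtualization is off, and a concurrency override allows one instance, stopping the oldest on retrigger.
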SoundCue /Skyfire/Sound/Skyfire_Backpack_Jump_Boost_Single_1P_Cue
root=QualityLevel_0
nodes [
 {id=Modulator_1, children [Random_0]}
 {id=Modulator_3, children [Random_2]}
 {id=QualityLevel_0, children [Modulator_1,Modulator_1,Modulator_3,Modulator_3]}
 {
  id=Random_0
  children [WavePlayer_1,WavePlayer_2,WavePlayer_3,WavePlayer_4,WavePlayer_5,WavePlayer_6]
  properties {bShouldExcludeFromBranchCulling=true, weights [1,1,1,1,1,1]}
 }
 {id=Random_2, children [WavePlayer_12, WavePlayer_13, WavePlayer_14], properties {bShouldExcludeFromBranchCulling=true, weights [1,1,1]}}
 {id=WavePlayer_1, properties {soundWaveAssetPtr=&/Skyfire_Backpack_Jump_Jetpack_Single_Boost_01}}
 {id=WavePlayer_12, properties {soundWaveAssetPtr=&Skyfire_Backpack_Jump_Jetpack_Single_Boost_01}}
 {id=WavePlayer_13, properties {soundWaveAssetPtr=&/Skyfire_Backpack_Jump_Jetpack_Single_Boost_02}}
 {id=WavePlayer_14, properties {soundWaveAssetPtr=&/Skyfire_Backpack_Jump_Jetpack_Single_Boost_03}}
 {id=WavePlayer_2, properties {soundWaveAssetPtr=&Skyfire_Backpack_Jump_Jetpack_Single_Boost_02}}
 {id=WavePlayer_3, properties {soundWaveAssetPtr=&Skyfire_Backpack_Jump_Jetpack_Single_Boost_03}}
 {id=WavePlayer_4, properties {soundWaveAssetPtr=&/Skyfire_Backpack_Jump_Jetpack_Single_Boost_04}}
 {id=WavePlayer_5, properties {soundWaveAssetPtr=&/Skyfire_Backpack_Jump_Jetpack_Single_Boost_05}}
 {id=WavePlayer_6, properties {soundWaveAssetPtr=&/Skyfire_Backpack_Jump_Jetpack_Single_Boost_06}}
]
properties {
 attenuationSettings=&/Skyfire_Backpack_Jump_Boost_Single_Att
 bBypassVolumeScaleForPriority=true
 bOverrideConcurrency=true
 maxDistance=6300
 priority=80
 soundClassObject=/Game/Sounds/FortSoundClassesAndModes/Fort_LTM_BR_InGame_Traversal_1P
 virtualizationMode=Disabled
 volumeMultiplier=1.2
 concurrencyOverrides {maxCount=1, resolutionRule=StopOldest}
}
