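# Sound cue for the single-boost sound of the Skyfire jetpack backpack jump, in the in-game traversal sound class.
# A quality-level node sends its first two slots to an equally weighted random pick of six boost waves and its last two to a pick of three.
# Each pick runs through a modulator at default settings for variation.
# Custom attenuation applies, virtualization is disabled, and a concurrency override allows two instances, stopping the quietest.
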SoundCue /Skyfire/Sound/Skyfire_Backpack_Jump_Boost_Single_Cue
root=QualityLevel_1
nodes [
 {id=Modulator_1, children [Random_3]}
 {id=Modulator_3, children [Random_4]}
 {id=QualityLevel_1, children [Modulator_1,Modulator_1,Modulator_3,Modulator_3]}
 {
  id=Random_3
  children [WavePlayer_4, WavePlayer_5, WavePlayer_6, WavePlayer_12, WavePlayer_13, WavePlayer_14]
  properties {bShouldExcludeFromBranchCulling=true, weights [1,1,1,1,1,1]}
 }
 {id=Random_4, children [WavePlayer_15, WavePlayer_16, WavePlayer_17], properties {bShouldExcludeFromBranchCulling=true, weights [1,1,1]}}
 {id=WavePlayer_12, properties {soundWaveAssetPtr=&/Skyfire_Backpack_Jump_Jetpack_Single_Boost_04}}
 {id=WavePlayer_13, properties {soundWaveAssetPtr=&/Skyfire_Backpack_Jump_Jetpack_Single_Boost_05}}
 {id=WavePlayer_14, properties {soundWaveAssetPtr=&/Skyfire_Backpack_Jump_Jetpack_Single_Boost_06}}
 {id=WavePlayer_15, properties {soundWaveAssetPtr=&/Skyfire_Backpack_Jump_Jetpack_Single_Boost_01}}
 {id=WavePlayer_16, properties {soundWaveAssetPtr=&/Skyfire_Backpack_Jump_Jetpack_Single_Boost_02}}
 {id=WavePlayer_17, properties {soundWaveAssetPtr=&/Skyfire_Backpack_Jump_Jetpack_Single_Boost_03}}
 {id=WavePlayer_4, properties {soundWaveAssetPtr=&Skyfire_Backpack_Jump_Jetpack_Single_Boost_01}}
 {id=WavePlayer_5, properties {soundWaveAssetPtr=&Skyfire_Backpack_Jump_Jetpack_Single_Boost_02}}
 {id=WavePlayer_6, properties {soundWaveAssetPtr=&Skyfire_Backpack_Jump_Jetpack_Single_Boost_03}}
]
properties {
 attenuationSettings=&/Skyfire_Backpack_Jump_Boost_Single_Att
 bOverrideConcurrency=true
 maxDistance=6300
 priority=70
 soundClassObject=/Game/Sounds/FortSoundClassesAndModes/Fort_LTM_BR_InGame_Traversal
 virtualizationMode=Disabled
 volumeMultiplier=1.2
 concurrencyOverrides {maxCount=2, resolutionRule=StopQuietest}
}
