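# Sound cue for the Skyfire backpack landing sound, built from borrowed Constructor stone landing footstep waves.
# A random node picks one of five equally weighted waves and feeds a modulator for variation.
# A quality-level root routes that chain to its first two slots and leaves the other two empty.
# It plays in the gadgets/items sound class with shared attenuation and a concurrency override allowing one instance, stopping the quietest.
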SoundCue /Skyfire/Sound/Skyfire_Backpack_Land_Cue
root=QualityLevel_0
nodes [
 {id=Modulator_0, children [Random_3]}
 {id=QualityLevel_0, children [Modulator_0,Modulator_0,null,null]}
 {id=Random_3, children [WavePlayer_15, WavePlayer_16, WavePlayer_17, WavePlayer_18, WavePlayer_19], properties {weights [1,1,1,1,1]}}
 {id=WavePlayer_15, properties {soundWaveAssetPtr=/FNGameplayCues/Sounds/Fort_Step_Sounds/Constructor/Constructor_Step_Land_Stone_06}}
 {id=WavePlayer_16, properties {soundWaveAssetPtr=/FNGameplayCues/Sounds/Fort_Step_Sounds/Constructor/Constructor_Step_Land_Stone_03}}
 {id=WavePlayer_17, properties {soundWaveAssetPtr=/FNGameplayCues/Sounds/Fort_Step_Sounds/Constructor/Constructor_Step_Land_Stone_05}}
 {id=WavePlayer_18, properties {soundWaveAssetPtr=/FNGameplayCues/Sounds/Fort_Step_Sounds/Constructor/Constructor_Step_Land_Stone_02}}
 {id=WavePlayer_19, properties {soundWaveAssetPtr=/FNGameplayCues/Sounds/Fort_Step_Sounds/Constructor/Constructor_Step_Land_Stone_04}}
]
properties {
 attenuationSettings=/Game/Sounds/Attenuation/Fort_Ability_Status_Effect_Activation_Att
 bOverrideConcurrency=true
 maxDistance=6000
 priority=0
 soundClassObject=&Fort_Gadgets_Items
 virtualizationMode=Disabled
 volumeMultiplier=0.4
 concurrencyOverrides {maxCount=1, resolutionRule=StopQuietest}
 concurrencySet [/Game/Sounds/Fort_GamePlay_Sounds/Player/LowGravityStones/LowGravityStone_JumpLand_Concc]
}
