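# Sound cue for the first-person Skyfire backpack materialize effect, judging by its name.
# A single wave player feeds the root directly, playing one materialize sound wave with no randomization or modulation.
# It routes to an in-game LTM sound class at boosted volume and raised priority, with virtualization disabled.
# A concurrency override allows only one instance, stopping the oldest when retriggered.
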SoundCue /Skyfire/Sound/Skyfire_Backpack_Materialize_1P_Cue
root=WavePlayer_0
nodes [{id=WavePlayer_0, properties {soundWaveAssetPtr=&/Skyfire_Backpack_Materialize_01}}]
properties {
 bBypassVolumeScaleForPriority=true
 bOverrideConcurrency=true
 maxDistance=2097152
 priority=75
 soundClassObject=/Game/Sounds/FortSoundClassesAndModes/Fort_LTM_BR_InGame_Master
 virtualizationMode=Disabled
 volumeMultiplier=1.35
 concurrencyOverrides {maxCount=1, resolutionRule=StopOldest}
}
