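# Sound cue for the materialize effect of the Skyfire backpack, routed through an in-game LTM master sound class.
# A single wave player feeds a modulator that randomizes pitch slightly below normal, under a quality-level root with only its first branch connected.
# It uses the backpack's attenuation asset, disables virtualization, and overrides concurrency to one instance, stopping the farthest before preventing new ones.
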SoundCue /Skyfire/Sound/Skyfire_Backpack_Materialize_Cue
root=QualityLevel_0
nodes [
 {id=Modulator_0, children [WavePlayer_0], properties {pitchMax=0.95, pitchMin=0.9}}
 {id=QualityLevel_0, children [Modulator_0,null,null,null]}
 {id=WavePlayer_0, properties {soundWaveAssetPtr=&/Skyfire_Backpack_Materialize_01}}
]
properties {
 attenuationSettings=&/Skyfire_Backpack_Att
 bOverrideConcurrency=true
 maxDistance=1350
 priority=60
 soundClassObject=/Game/Sounds/FortSoundClassesAndModes/Fort_LTM_BR_InGame_Master
 virtualizationMode=Disabled
 volumeMultiplier=0.4
 concurrencyOverrides {maxCount=1, resolutionRule=StopFarthestThenPreventNew}
}
