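# Skyfire sound cue for the first-person backpack rarity upgrade sound, routed to the Battle Royale LTM special-item 1P sound class.
# A quality-level node feeds a short fixed delay into a single wave player, with only its first quality slot connected.
# Concurrency override allows one instance, stopping the oldest.
# The cue has boosted volume, raised priority that ignores volume scaling, and virtualization disabled.
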SoundCue /Skyfire/Sound/Skyfire_Backpack_Rarity_Upgrade_1P_Cue
root=QualityLevel_0
nodes [
 {id=Delay_0, children [WavePlayer_1], properties {delayMax=0.2, delayMin=0.2}}
 {id=QualityLevel_0, children [Delay_0,null,null,null]}
 {id=WavePlayer_1, properties {soundWaveAssetPtr=&/Skyfire_Backpack_Rarity_Upgrade_01}}
]
properties {
 bBypassVolumeScaleForPriority=true
 bOverrideConcurrency=true
 maxDistance=2097152
 priority=35
 soundClassObject=/Game/Sounds/FortSoundClassesAndModes/Fort_LTM_BR_InGame_Special_Item_1P
 virtualizationMode=Disabled
 volumeMultiplier=2
 concurrencyOverrides {maxCount=1, resolutionRule=StopOldest}
}
