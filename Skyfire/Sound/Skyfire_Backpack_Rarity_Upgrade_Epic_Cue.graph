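# Sound cue named for the Skyfire backpack's upgrade to Epic rarity, holding only playback settings.
# The node graph is empty with no root, so the cue references no audio as saved.
# It uses the shared backpack attenuation and Fortnite's LTM special-item sound class at reduced volume, with virtualization disabled.
# A concurrency override limits it to a single instance at a time.
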SoundCue /Skyfire/Sound/Skyfire_Backpack_Rarity_Upgrade_Epic_Cue
root=null
nodes []
properties {
 attenuationSettings=&/Skyfire_Backpack_Att
 bOverrideConcurrency=true
 maxDistance=1350
 soundClassObject=/Game/Sounds/FortSoundClassesAndModes/Fort_LTM_BR_InGame_Special_Item
 virtualizationMode=Disabled
 volumeMultiplier=0.5
 concurrencyOverrides {maxCount=1}
}
