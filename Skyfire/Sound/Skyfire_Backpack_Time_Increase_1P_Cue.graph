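# Sound cue for the first-person sound of a time increase on the Skyfire backpack, routed to an LTM special-item 1P sound class.
# A single wave player feeds the mechanical timer-increase wave straight to the output, with boosted volume and virtualization disabled.
# Overridden concurrency allows one instance at a time, stopping the oldest so repeated increases retrigger rather than stack.
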SoundCue /Skyfire/Sound/Skyfire_Backpack_Time_Increase_1P_Cue
root=WavePlayer_0
nodes [{id=WavePlayer_0, properties {soundWaveAssetPtr=&/Skyfire_Backpack_Timer_Increase_Mech_01}}]
properties {
 bBypassVolumeScaleForPriority=true
 bOverrideConcurrency=true
 maxDistance=2097152
 priority=35
 soundClassObject=/Game/Sounds/FortSoundClassesAndModes/Fort_LTM_BR_InGame_Special_Item_1P
 virtualizationMode=Disabled
 volumeMultiplier=1.85
 attenuationOverrides {
  attenuationShape=Cone
  bEnableOcclusion=true
  falloffDistance=6000
  occlusionLowPassFilterFrequency=3000
  attenuationShapeExtents {x=0, z=20}
 }
 concurrencyOverrides {maxCount=1, resolutionRule=StopOldest}
}
