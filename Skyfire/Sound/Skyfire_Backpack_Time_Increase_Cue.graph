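# Sound cue for the Skyfire backpack's time increase, playing a mechanical timer sound under the LTM in-game special item sound class.
# A quality level node feeds a single wave player from its first slot only, with volume and pitch slightly lowered.
# Spatialization references the shared Skyfire backpack attenuation asset.
# Concurrency is overridden to allow one instance, stopping the quietest, and virtualization is disabled.
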SoundCue /Skyfire/Sound/Skyfire_Backpack_Time_Increase_Cue
root=QualityLevel_0
nodes [
 {id=QualityLevel_0, children [WavePlayer_0,null,null,null]}
 {id=WavePlayer_0, properties {soundWaveAssetPtr=&/Skyfire_Backpack_Timer_Increase_Mech_01}}
]
properties {
 attenuationSettings=&/Skyfire_Backpack_Att
 bOverrideConcurrency=true
 maxDistance=1350
 pitchMultiplier=0.9
 priority=0
 soundClassObject=/Game/Sounds/FortSoundClassesAndModes/Fort_LTM_BR_InGame_Special_Item
 virtualizationMode=Disabled
 volumeMultiplier=0.85
 attenuationOverrides {
  attenuationShape=Cone
  bEnableOcclusion=true
  falloffDistance=6000
  occlusionLowPassFilterFrequency=3000
  attenuationShapeExtents {x=0, z=20}
 }
 concurrencyOverrides {maxCount=1, resolutionRule=StopQuietest}
}
