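# Sound cue for an idle computer ambience in Skyfire: a looping buzz bed layered with intermittent comms walla chatter.
# A mixer blends the quieter looping buzz wave with a looped branch that waits a random delay, then plays a randomly chosen, equally weighted walla clip.
# A modulator varies each chatter clip's pitch and volume, so the layer repeats indefinitely without sounding identical.
# It uses the computer idle attenuation asset and an in-game ambience sound class.
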SoundCue /Skyfire/Sound/Skyfire_Computer_Idle_Activity_Buzz_Loop_A_Cue
root=Mixer_0
nodes [
 {id=Delay_0, children [Random_1], properties {delayMax=5, delayMin=1}}
 {id=Looping_0, children [Modulator_0]}
 {id=Mixer_0, children [WavePlayer_0,Looping_0], properties {inputVolume [0.3,1]}}
 {id=Modulator_0, children [Delay_0], properties {pitchMax=1.25, volumeMin=0.75}}
 {
  id=Random_1
  children [
   WavePlayer_1
   WavePlayer_2
   WavePlayer_3
   WavePlayer_4
   WavePlayer_5
   WavePlayer_6
   WavePlayer_7
   WavePlayer_8
   WavePlayer_9
   WavePlayer_10
   WavePlayer_11
   WavePlayer_12
   WavePlayer_13
   WavePlayer_14
   WavePlayer_15
   WavePlayer_16
   WavePlayer_17
   WavePlayer_18
   WavePlayer_19
   WavePlayer_20
   WavePlayer_21
   WavePlayer_22
  ]
  properties {weights [1,1,1,1,1,1,1,1,1,1,1,1,1,1,1,1,1,1,1,1,1,1]}
 }
 {id=WavePlayer_0, properties {bLooping=true, soundWaveAssetPtr=&/Skyfire_Computer_Idle_Activity_Buzz_Loop_01}}
 {id=WavePlayer_1, properties {soundWaveAssetPtr=&/Skyfire_Computer_Comms_Walla_10}}
 {id=WavePlayer_10, properties {soundWaveAssetPtr=&/Skyfire_Computer_Comms_Walla_19}}
 {id=WavePlayer_11, properties {soundWaveAssetPtr=&/Skyfire_Computer_Comms_Walla_20}}
 {id=WavePlayer_12, properties {soundWaveAssetPtr=&/Skyfire_Computer_Comms_Walla_21}}
 {id=WavePlayer_13, properties {soundWaveAssetPtr=&/Skyfire_Computer_Comms_Walla_22}}
 {id=WavePlayer_14, properties {soundWaveAssetPtr=&/Skyfire_Computer_Comms_Walla_01}}
 {id=WavePlayer_15, properties {soundWaveAssetPtr=&/Skyfire_Computer_Comms_Walla_02}}
 {id=WavePlayer_16, properties {soundWaveAssetPtr=&/Skyfire_Computer_Comms_Walla_03}}
 {id=WavePlayer_17, properties {soundWaveAssetPtr=&/Skyfire_Computer_Comms_Walla_04}}
 {id=WavePlayer_18, properties {soundWaveAssetPtr=&/Skyfire_Computer_Comms_Walla_05}}
 {id=WavePlayer_19, properties {soundWaveAssetPtr=&/Skyfire_Computer_Comms_Walla_06}}
 {id=WavePlayer_2, properties {soundWaveAssetPtr=&/Skyfire_Computer_Comms_Walla_11}}
 {id=WavePlayer_20, properties {soundWaveAssetPtr=&/Skyfire_Computer_Comms_Walla_07}}
 {id=WavePlayer_21, properties {soundWaveAssetPtr=&/Skyfire_Computer_Comms_Walla_08}}
 {id=WavePlayer_22, properties {soundWaveAssetPtr=&/Skyfire_Computer_Comms_Walla_09}}
 {id=WavePlayer_3, properties {soundWaveAssetPtr=&/Skyfire_Computer_Comms_Walla_12}}
 {id=WavePlayer_4, properties {soundWaveAssetPtr=&/Skyfire_Computer_Comms_Walla_13}}
 {id=WavePlayer_5, properties {soundWaveAssetPtr=&/Skyfire_Computer_Comms_Walla_14}}
 {id=WavePlayer_6, properties {soundWaveAssetPtr=&/Skyfire_Computer_Comms_Walla_15}}
 {id=WavePlayer_7, properties {soundWaveAssetPtr=&/Skyfire_Computer_Comms_Walla_16}}
 {id=WavePlayer_8, properties {soundWaveAssetPtr=&/Skyfire_Computer_Comms_Walla_17}}
 {id=WavePlayer_9, properties {soundWaveAssetPtr=&/Skyfire_Computer_Comms_Walla_18}}
]
properties {
 attenuationSettings=&/Skyfire_Computer_Idle_Att
 maxDistance=3800
 soundClassObject=/Game/Sounds/FortSoundClassesAndModes/Fort_LTM_BR_InGame_Ambience
 volumeMultiplier=1
}
