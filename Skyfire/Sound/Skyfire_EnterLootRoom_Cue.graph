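# Sound cue named for entering a loot room in Skyfire, but it has no root or sound nodes, so it plays nothing as it stands.
# It routes to the LTM Battle Royale in-game master sound class.
# It stores cone attenuation with occlusion and per-owner, stop-oldest concurrency settings, though neither override is shown as enabled.
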SoundCue /Skyfire/Sound/Skyfire_EnterLootRoom_Cue
root=null
nodes []
properties {
 maxDistance=2097152
 soundClassObject=/Game/Sounds/FortSoundClassesAndModes/Fort_LTM_BR_InGame_Master
 volumeMultiplier=1
 attenuationOverrides {
  attenuationShape=Cone
  bEnableOcclusion=true
  falloffDistance=6000
  occlusionLowPassFilterFrequency=3000
  attenuationShapeExtents {x=0, z=20}
 }
 concurrencyOverrides {bLimitToOwner=true, maxCount=3, resolutionRule=StopOldest}
}
