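# Sound cue for a looping, spatialized Skyfire exhaust pipe ambience.
# A looping wave player feeds a modulator that randomizes pitch, then an enveloper that fades volume in over about one second.
# A quality-level root routes that same chain to its first two slots and leaves the other two empty.
# The cue boosts volume, uses a dedicated attenuation asset and overrides concurrency to cap simultaneous instances at two.
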SoundCue /Skyfire/Sound/Skyfire_Exhaust_Pipe_Loop_Cue
root=QualityLevel_0
nodes [
 {
  id=Enveloper_1
  children [Modulator_0]
  properties {pitchCurve {editorCurveData {keys [{time=0}]}}, volumeCurve {editorCurveData {keys [<at -0.0001 0>,<at 1 1>]}}}
 }
 {id=Modulator_0, children [WavePlayer_0], properties {pitchMax=1.2}}
 {id=QualityLevel_0, children [Enveloper_1,Enveloper_1,null,null]}
 {id=WavePlayer_0, properties {bLooping=true, soundWaveAssetPtr=&/Skyfire_Exhaust_Pipe_Loop}}
]
properties {
 attenuationSettings=&/Skyfire_Exhaust_Pipe_Loop_Att
 bOverrideConcurrency=true
 maxDistance=1298.9999
 priority=10
 soundClassObject=/Game/Sounds/FortSoundClassesAndModes/Fort_LTM_BR_InGame_Phase_Main_Amb_3D
 volumeMultiplier=4
 concurrencyOverrides {maxCount=2}
}
