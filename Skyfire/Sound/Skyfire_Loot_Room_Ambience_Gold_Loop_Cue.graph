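# Sound cue for the looping gold loot room ambience in Skyfire.
# A looping wave player feeds an Enveloper root that fades volume in over the first quarter second.
# It uses an LTM phase-end 2D ambience sound class, with boosted volume, a very large max distance and a high priority that bypasses volume scaling.
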SoundCue /Skyfire/Sound/Skyfire_Loot_Room_Ambience_Gold_Loop_Cue
root=Enveloper_0
nodes [
 {
  id=Enveloper_0
  children [WavePlayer_0]
  properties {pitchCurve {editorCurveData {keys [{time=0}]}}, volumeCurve {editorCurveData {keys [<at 0 0>,<at 0.25 1>]}}}
 }
 {id=WavePlayer_0, properties {bLooping=true, soundWaveAssetPtr=&/Skyfire_Loot_Room_Ambience_Loop}}
]
properties {
 bBypassVolumeScaleForPriority=true
 maxDistance=2097152
 priority=90
 soundClassObject=/Game/Sounds/FortSoundClassesAndModes/Fort_LTM_BR_InGame_Phase_End_Amb_2D_PSM
 volumeMultiplier=1.5
}
