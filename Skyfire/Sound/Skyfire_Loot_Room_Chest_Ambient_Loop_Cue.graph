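# Looping 3D ambient sound cue, named for the Skyfire loot room chest, that plays the chest loop wave.
# The wave gets slight pitch variation, a slow oscillating volume modulation and an envelope that ramps volume in, with the first two quality levels sharing this chain and the other two empty.
# An external attenuation asset handles distance behavior, while overridden concurrency caps instances at three and stops the quietest.
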SoundCue /Skyfire/Sound/Skyfire_Loot_Room_Chest_Ambient_Loop_Cue
root=QualityLevel_0
nodes [
 {
  id=Enveloper_0
  children [Oscillator_0]
  properties {pitchCurve {editorCurveData {keys [{time=0}]}}, volumeCurve {editorCurveData {keys [<at 0 0>,<at 1 1>]}}}
 }
 {id=Modulator_0, children [WavePlayer_2], properties {pitchMax=1.01, pitchMin=0.99}}
 {
  id=Oscillator_0
  children [Modulator_0]
  properties {amplitudeMax=0.1, amplitudeMin=0.05, bModulateVolume=true, centerMax=0.5, centerMin=0.2, frequencyMax=0.5, frequencyMin=0.1}
 }
 {id=QualityLevel_0, children [Enveloper_0,Enveloper_0,null,null]}
 {id=WavePlayer_2, properties {bLooping=true, soundWaveAssetPtr=&/Skyfire_Loot_Room_Chest_Loop}}
]
properties {
 attenuationSettings=&/Skyfire_Loot_Room_Chest_Amb_Att
 bOverrideConcurrency=true
 maxDistance=600
 priority=25
 soundClassObject=/Game/Sounds/FortSoundClassesAndModes/Fort_LTM_BR_InGame_Phase_End_AMB_3D
 volumeMultiplier=1
 attenuationOverrides {
  bEnableFocusInterpolation=true
  bEnableListenerFocus=true
  bEnableOcclusion=true
  distanceAlgorithm=Logarithmic
  falloffDistance=1100
  focusAttackInterpSpeed=20
  focusAzimuth=20
  focusReleaseInterpSpeed=10
  focusVolumeAttenuation=1.75
  lPFFrequencyAtMax=10000
  lPFRadiusMax=1000
  lPFRadiusMin=500
  nonFocusAzimuth=40
  nonFocusVolumeAttenuation=0.8
  occlusionInterpolationTime=0.3
  occlusionLowPassFilterFrequency=3000
  occlusionVolumeAttenuation=0.75
  spatializationAlgorithm=SPATIALIZATION_HRTF
  stereoSpread=512
  attenuationShapeExtents {x=0}
  customAttenuationCurve {editorCurveData {keys [<at 0 1 RCIM_Cubic>, {arriveTangent=-1, interpMode=RCIM_Cubic, leaveTangent=-1, time=0.5, value=0.1}, <at 1 0 RCIM_Cubic>]}}
 }
 concurrencyOverrides {maxCount=3, resolutionRule=StopQuietest, voiceStealReleaseTime=0.1}
}
