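# Sound cue for a loot room chest opening in Skyfire, layering a latch sound with a magic flourish.
# A Mixer combines a Modulator-varied latch wave with one of four equally weighted random magic waves routed through a Delay node, with the latch mixed louder.
# It plays spatialized through its own attenuation asset, with virtualization disabled and concurrency overridden to cap simultaneous instances by stopping the quietest.
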SoundCue /Skyfire/Sound/Skyfire_Loot_Room_Chest_Open_Cue
root=Mixer_0
nodes [
 {id=Delay_0, children [Random_0]}
 {id=Mixer_0, children [Delay_0,Modulator_1], properties {inputVolume [1,1.25]}}
 {id=Modulator_1, children [WavePlayer_1]}
 {
  id=Random_0
  children [WavePlayer_2,WavePlayer_4,WavePlayer_5,WavePlayer_6]
  properties {bShouldExcludeFromBranchCulling=true, weights [1,1,1,1]}
 }
 {id=WavePlayer_1, properties {soundWaveAssetPtr=&/Skyfire_Loot_Room_Chest_Open_Latch_01}}
 {id=WavePlayer_2, properties {soundWaveAssetPtr=&/Skyfire_Loot_Room_Chest_Open_Magic_01}}
 {id=WavePlayer_4, properties {soundWaveAssetPtr=&/Skyfire_Loot_Room_Chest_Open_Magic_02}}
 {id=WavePlayer_5, properties {soundWaveAssetPtr=&/Skyfire_Loot_Room_Chest_Open_Magic_03}}
 {id=WavePlayer_6, properties {soundWaveAssetPtr=&/Skyfire_Loot_Room_Chest_Open_Magic_04}}
]
properties {
 attenuationSettings=&/Skyfire_Loot_Room_Chest_Open_Att
 bOverrideConcurrency=true
 maxDistance=1100
 priority=85
 soundClassObject=/Game/Sounds/FortSoundClassesAndModes/Fort_LTM_BR_InGame_Phase_End_AMB_3D
 virtualizationMode=Disabled
 volumeMultiplier=1.35
 attenuationOverrides {bEnableOcclusion=true, falloffDistance=1000, attenuationShapeExtents {x=500}}
 concurrencyOverrides {maxCount=2, resolutionRule=StopQuietest, voiceStealReleaseTime=0.2}
}
