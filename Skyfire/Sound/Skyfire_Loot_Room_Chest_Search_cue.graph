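# Sound cue for searching a Skyfire loot room chest, played as attenuated positional audio.
# A quality-level switch picks among four branches: two mix a fading treasure-search layer with gear-rummage foley, and two play only the optimized treasure-search wave through modulators.
# The first branch randomizes among three gear waves, while the second uses a single one.
# Concurrency is capped at two instances, stopping the quietest.
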SoundCue /Skyfire/Sound/Skyfire_Loot_Room_Chest_Search_cue
root=QualityLevel_0
nodes [
 {
  id=Enveloper_0
  children [WavePlayer_3]
  properties {
   pitchCurve {editorCurveData {keys [{time=0}]}}
   volumeCurve {editorCurveData {keys [
      {arriveTangent=-0.0387, time=0, value=1}
      {arriveTangent=-0.264, interpMode=RCIM_Cubic, leaveTangent=-0.264, time=2.7563, value=0.8772}
      <at 3.7875 0>
     ]}}
  }
 }
 {
  id=Enveloper_1
  children [WavePlayer_2]
  properties {
   pitchCurve {editorCurveData {keys [{time=0}]}}
   volumeCurve {editorCurveData {keys [
      {arriveTangent=-0.0387, time=0, value=1}
      {arriveTangent=-0.264, interpMode=RCIM_Cubic, leaveTangent=-0.264, time=2.7563, value=0.8772}
      <at 3.7875 0>
     ]}}
  }
 }
 {id=Mixer_0, children [Enveloper_0,WavePlayer_4], properties {inputVolume [0.6,1]}}
 {id=Mixer_2, children [Enveloper_1,Random_13], properties {inputVolume [0.6,1]}}
 {id=Modulator_0, children [WavePlayer_0]}
 {id=Modulator_1, children [WavePlayer_1]}
 {id=QualityLevel_0, children [Mixer_2,Mixer_0,Modulator_0,Modulator_1]}
 {id=Random_13, children [WavePlayer_58, WavePlayer_59, WavePlayer_60], properties {weights [1,1,1]}}
 {id=WavePlayer_0, properties {soundWaveAssetPtr=/Game/Sounds/Foley_Loot/Containers/Treasure_Chest/treasure_search_opt}}
 {id=WavePlayer_1, properties {soundWaveAssetPtr=&treasure_search_opt}}
 {id=WavePlayer_2, properties {soundWaveAssetPtr=/Game/Sounds/Foley_Loot/Containers/Treasure_Chest/treasure_search}}
 {id=WavePlayer_3, properties {soundWaveAssetPtr=&treasure_search}}
 {id=WavePlayer_4, properties {soundWaveAssetPtr=/Game/Sounds/Foley_Loot/Containers/Misc/Prop_Gear_Search_04}}
 {id=WavePlayer_58, properties {soundWaveAssetPtr=&Prop_Gear_Search_04}}
 {id=WavePlayer_59, properties {soundWaveAssetPtr=/Game/Sounds/Foley_Loot/Containers/Misc/Prop_Gear_Search_06}}
 {id=WavePlayer_60, properties {soundWaveAssetPtr=/Game/Sounds/Foley_Loot/Containers/Misc/Prop_Gear_Search_01}}
]
properties {
 attenuationSettings=&/Skyfire_Loot_Room_Chest_Open_Att
 bOverrideConcurrency=true
 maxDistance=1100
 priority=25
 soundClassObject=/Game/Sounds/FortSoundClassesAndModes/Fort_LTM_BR_InGame_Phase_End_AMB_3D
 volumeMultiplier=1.25
 concurrencyOverrides {maxCount=2, resolutionRule=StopQuietest}
}
