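# Sound cue for the first-person alarm of the Skyfire mothership purge alert.
# A fixed short delay feeds a single wave player holding the alarm sound.
# It plays boosted at top priority without virtualization, and its concurrency override allows one instance, blocking new ones while it plays.
# It routes through an LTM in-game sound class and sets a very large max distance.
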SoundCue /Skyfire/Sound/Skyfire_Mothership_Purge_Alert_1P_Cue
root=Delay_0
nodes [
 {id=Delay_0, children [WavePlayer_0], properties {delayMax=0.45, delayMin=0.45}}
 {id=WavePlayer_0, properties {soundWaveAssetPtr=&/Skyfire_Mothership_Purge_Alert_Alarm_1P_01}}
]
properties {
 bBypassVolumeScaleForPriority=true
 bOverrideConcurrency=true
 maxDistance=2097152
 priority=100
 soundClassObject=/Game/Sounds/FortSoundClassesAndModes/Fort_LTM_BR_InGame_Exit_PSM
 virtualizationMode=Disabled
 volumeMultiplier=3
 concurrencyOverrides {maxCount=1, resolutionRule=PreventNew}
}
