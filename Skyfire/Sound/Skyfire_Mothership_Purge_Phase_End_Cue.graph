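# Sound cue for the end of the Skyfire Mothership purge phase.
# It plays a single wave after a short fixed delay, at reduced volume, using the Purge Player attenuation asset and an in-game warp third-person sound class.
# A concurrency override allows only one instance and rejects new plays while it sounds, and virtualization is disabled.
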SoundCue /Skyfire/Sound/Skyfire_Mothership_Purge_Phase_End_Cue
root=Delay_0
nodes [
 {id=Delay_0, children [WavePlayer_0], properties {delayMax=0.25, delayMin=0.25}}
 {id=WavePlayer_0, properties {soundWaveAssetPtr=&/Skyfire_Mothership_Purge_Phase_End_01}}
]
properties {
 attenuationSettings=&/Skyfire_Mothership_Purge_Player_Att
 bOverrideConcurrency=true
 maxDistance=2097152
 soundClassObject=/Game/Sounds/FortSoundClassesAndModes/Fort_LTM_BR_InGame_Warp_3P
 virtualizationMode=Disabled
 volumeMultiplier=0.25
 concurrencyOverrides {maxCount=1, resolutionRule=PreventNew}
}
