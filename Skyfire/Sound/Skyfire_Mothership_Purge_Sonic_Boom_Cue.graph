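# Sound cue for the Skyfire mothership purge sonic boom, using a player attenuation asset and an LTM in-game third-person warp sound class.
# A Random node picks one of four equally weighted supersonic boom waves, which feeds a Modulator that varies pitch while holding volume constant.
# Overridden concurrency allows one instance and blocks new ones while it plays, with virtualization disabled and an overall reduced volume.
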SoundCue /Skyfire/Sound/Skyfire_Mothership_Purge_Sonic_Boom_Cue
root=Modulator_0
nodes [
 {id=Modulator_0, children [Random_0], properties {pitchMax=1.25, volumeMax=1, volumeMin=1}}
 {
  id=Random_0
  children [WavePlayer_1,WavePlayer_2,WavePlayer_3,WavePlayer_4]
  properties {bShouldExcludeFromBranchCulling=true, weights [1,1,1,1]}
 }
 {id=WavePlayer_1, properties {soundWaveAssetPtr=&/Skyfire_Mothership_Purge_SuperSonic_Boom_01}}
 {id=WavePlayer_2, properties {soundWaveAssetPtr=&/Skyfire_Mothership_Purge_SuperSonic_Boom_02}}
 {id=WavePlayer_3, properties {soundWaveAssetPtr=&/Skyfire_Mothership_Purge_SuperSonic_Boom_03}}
 {id=WavePlayer_4, properties {soundWaveAssetPtr=&/Skyfire_Mothership_Purge_SuperSonic_Boom_04}}
]
properties {
 attenuationSettings=&/Skyfire_Mothership_Purge_Player_Att
 bOverrideConcurrency=true
 maxDistance=2097152
 soundClassObject=/Game/Sounds/FortSoundClassesAndModes/Fort_LTM_BR_InGame_Warp_3P
 virtualizationMode=Disabled
 volumeMultiplier=0.35
 concurrencyOverrides {maxCount=1, resolutionRule=PreventNew}
}
