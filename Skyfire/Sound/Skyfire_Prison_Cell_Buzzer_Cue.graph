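# Sound cue for the Skyfire prison cell buzzer, a one-shot style wave played after a short fixed delay.
# A single wave player feeds a Delay node at the root, routed through an LTM phase-start 3D ambience sound class.
# Boosted volume, high priority, disabled virtualization and a concurrency override allow only one instance at a time, with new plays blocked while it sounds.
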SoundCue /Skyfire/Sound/Skyfire_Prison_Cell_Buzzer_Cue
root=Delay_0
nodes [
 {id=Delay_0, children [WavePlayer_0], properties {delayMax=0.5, delayMin=0.5}}
 {id=WavePlayer_0, properties {soundWaveAssetPtr=&/Skyfire_Prison_Cell_Buzzer}}
]
properties {
 bBypassVolumeScaleForPriority=true
 bOverrideConcurrency=true
 maxDistance=2097152
 priority=90
 soundClassObject=/Game/Sounds/FortSoundClassesAndModes/Fort_LTM_BR_InGame_Phase_Start_Amb_3D
 virtualizationMode=Disabled
 volumeMultiplier=1.5
 concurrencyOverrides {maxCount=1, resolutionRule=PreventNew}
}
