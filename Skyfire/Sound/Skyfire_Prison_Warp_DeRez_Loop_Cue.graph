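# Sound cue for the Skyfire prison warp de-rez loop, played as a spatialized special-item sound with character de-rez attenuation.
# A Mixer layers a slightly boosted, non-looping start tell over a looping de-rez wave that an Enveloper fades in across 1.5 seconds.
# Each layer has its own Modulator for random pitch variation.
# A concurrency override caps instances at three, stopping the quietest, and virtualization is disabled.
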SoundCue /Skyfire/Sound/Skyfire_Prison_Warp_DeRez_Loop_Cue
root=Mixer_0
nodes [
 {
  id=Enveloper_0
  children [WavePlayer_0]
  properties {
   pitchCurve {editorCurveData {keys [{time=0}]}}
   volumeCurve {editorCurveData {keys [{arriveTangent=0.1339, interpMode=RCIM_Cubic, leaveTangent=0.1339, tangentMode=RCTM_User, time=0, value=0}, <at 1.5 1>]}}
  }
 }
 {id=Mixer_0, children [Modulator_1,Modulator_0], properties {inputVolume [1.2,1]}}
 {id=Modulator_0, children [Enveloper_0], properties {pitchMax=1.2}}
 {id=Modulator_1, children [WavePlayer_1], properties {pitchMax=1.2}}
 {id=WavePlayer_0, properties {bLooping=true, soundWaveAssetPtr=&/Skyfire_Warp_DeRez_Loop}}
 {id=WavePlayer_1, properties {soundWaveAssetPtr=&/Skyfire_Warp_DeRez_Tell_Loop_Start_01}}
]
properties {
 attenuationSettings=&/Skyfire_Warp_Character_DeRez_Loop_Att
 bOverrideConcurrency=true
 maxDistance=6200
 priority=85
 soundClassObject=/Game/Sounds/FortSoundClassesAndModes/Fort_LTM_BR_InGame_Special_Item
 virtualizationMode=Disabled
 volumeMultiplier=1.65
 concurrencyOverrides {maxCount=3, resolutionRule=StopQuietest}
}
