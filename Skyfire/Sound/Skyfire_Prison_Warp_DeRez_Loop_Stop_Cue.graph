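# Sound cue for the stop of the Skyfire prison warp de-rez loop, playing a single loop-stop wave.
# A modulator wraps the wave player to vary pitch on each play, with a raised upper pitch limit.
# It uses the character de-rez loop attenuation and an in-game LTM sound class, with boosted volume and virtualization disabled.
# A concurrency override caps simultaneous instances, stopping the quietest.
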SoundCue /Skyfire/Sound/Skyfire_Prison_Warp_DeRez_Loop_Stop_Cue
root=Modulator_2
nodes [
 {id=Modulator_2, children [WavePlayer_2], properties {pitchMax=1.2}}
 {id=WavePlayer_2, properties {soundWaveAssetPtr=&/Skyfire_Warp_DeRez_Tell_Loop_Stop_01}}
]
properties {
 attenuationSettings=&/Skyfire_Warp_Character_DeRez_Loop_Att
 bOverrideConcurrency=true
 maxDistance=6200
 priority=85
 soundClassObject=/Game/Sounds/FortSoundClassesAndModes/Fort_LTM_BR_InGame_Master
 virtualizationMode=Disabled
 volumeMultiplier=1.5
 concurrencyOverrides {maxCount=3, resolutionRule=StopQuietest}
}
