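# Sound cue for the first-person warp into the Skyfire prison arena, playing a single warp de-rez transition explosion wave.
# The wave sits behind a fixed 0.6-second delay, wired only to the first slot of a quality-level node; the other slots are empty.
# An overridden concurrency rule allows one instance and blocks new ones, with virtualization disabled and routing to a 2D ambience LTM sound class.
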SoundCue /Skyfire/Sound/Skyfire_Prison_Warp_Enter_Arena_1P_Cue
root=QualityLevel_0
nodes [
 {id=Delay_0, children [WavePlayer_0], properties {delayMax=0.6, delayMin=0.6}}
 {id=QualityLevel_0, children [Delay_0,null,null,null]}
 {id=WavePlayer_0, properties {soundWaveAssetPtr=&/Skyfire_Warp_DeRez_Exit_Transition_Explo_1P_01}}
]
properties {
 bOverrideConcurrency=true
 maxDistance=2097152
 priority=25
 soundClassObject=/Game/Sounds/FortSoundClassesAndModes/Fort_LTM_BR_InGame_Phase_Main_Amb_2D_PSM
 virtualizationMode=Disabled
 volumeMultiplier=1
 concurrencyOverrides {maxCount=1, resolutionRule=PreventNew}
}
