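# Sound cue for Skyfire's prison warp into the arena, judging by its name, built from warp de-rez exit transition explosion waves.
# A Random node picks one of four equally weighted variations and feeds it through a Modulator.
# It uses the warp character attenuation and an LTM ambient 3D sound class.
# A concurrency override caps it at two instances, stopping the quietest, with virtualization disabled.
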SoundCue /Skyfire/Sound/Skyfire_Prison_Warp_Enter_Arena_Cue
root=Modulator_0
nodes [
 {id=Modulator_0, children [Random_0]}
 {id=Random_0, children [WavePlayer_1,WavePlayer_2,WavePlayer_3,WavePlayer_4], properties {weights [1,1,1,1]}}
 {id=WavePlayer_1, properties {soundWaveAssetPtr=&/Skyfire_Warp_DeRez_Exit_Transition_Explo_01}}
 {id=WavePlayer_2, properties {soundWaveAssetPtr=&/Skyfire_Warp_DeRez_Exit_Transition_Explo_02}}
 {id=WavePlayer_3, properties {soundWaveAssetPtr=&/Skyfire_Warp_DeRez_Exit_Transition_Explo_03}}
 {id=WavePlayer_4, properties {soundWaveAssetPtr=&/Skyfire_Warp_DeRez_Exit_Transition_Explo_04}}
]
properties {
 attenuationSettings=&/Skyfire_Warp_Character_Att
 bBypassVolumeScaleForPriority=true
 bOverrideConcurrency=true
 maxDistance=6200
 priority=30
 soundClassObject=/Game/Sounds/FortSoundClassesAndModes/Fort_LTM_BR_InGame_Phase_Main_Amb_3D
 virtualizationMode=Disabled
 volumeMultiplier=0.85
 concurrencyOverrides {maxCount=2, resolutionRule=StopQuietest}
}
