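# Sound cue for the first-person start of the Skyfire warp de-rez effect.
# A single wave player feeds the root, playing one start sample at reduced volume through a dedicated first-person warp sound class.
# It has high priority that ignores volume scaling, with virtualization disabled.
# A concurrency override allows one instance at a time and blocks new plays while it is active.
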SoundCue /Skyfire/Sound/Skyfire_Warp_DeRez_Start_1P_Cue
root=WavePlayer_0
nodes [{id=WavePlayer_0, properties {soundWaveAssetPtr=&/Skyfire_Warp_DeRez_Start_01}}]
properties {
 bBypassVolumeScaleForPriority=true
 bOverrideConcurrency=true
 maxDistance=2097152
 priority=90
 soundClassObject=/Game/Sounds/FortSoundClassesAndModes/Fort_LTM_BR_InGame_Warp_1P_PSM
 virtualizationMode=Disabled
 volumeMultiplier=0.195
 concurrencyOverrides {maxCount=1, resolutionRule=PreventNew}
}
