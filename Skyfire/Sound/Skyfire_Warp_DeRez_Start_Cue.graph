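# Sound cue for the start of a Skyfire warp de-rez effect, going by its name.
# The node graph is empty with no root, so as saved it plays no audio.
# It keeps a very large max distance and routes to the Fort LTM BR in-game warp sound class.
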SoundCue /Skyfire/Sound/Skyfire_Warp_DeRez_Start_Cue
root=null
nodes []
properties {maxDistance=2097152, soundClassObject=/Game/Sounds/FortSoundClassesAndModes/Fort_LTM_BR_InGame_Warp}
