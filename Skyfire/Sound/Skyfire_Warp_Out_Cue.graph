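# Sound cue for the Skyfire warp-out effect, routed to the in-game warp LTM sound class.
# A single wave player feeds the Skyfire_Warp_Out_01 wave straight to the output at slightly reduced volume.
# Concurrency override allows one instance at a time, stopping the oldest when a new one starts.
# It has raised priority, a very large max distance and virtualization disabled.
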SoundCue /Skyfire/Sound/Skyfire_Warp_Out_Cue
root=WavePlayer_0
nodes [{id=WavePlayer_0, properties {soundWaveAssetPtr=&/Skyfire_Warp_Out_01}}]
properties {
 bOverrideConcurrency=true
 maxDistance=2097152
 priority=50
 soundClassObject=/Game/Sounds/FortSoundClassesAndModes/Fort_LTM_BR_InGame_Warp
 virtualizationMode=Disabled
 volumeMultiplier=0.9
 concurrencyOverrides {maxCount=1, resolutionRule=StopOldest}
}
